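# Impact sound cue for the Hightower Date pickaxe cosmetic.
# A mixer layers a random pick from four Hightower Date impacts with a half-volume random pick from five reused Galileo Ferry pickaxe impacts.
# Each layer has its own modulator, and pickaxe attenuation, impact sound class and player melee concurrency settings apply.
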SoundCue /BRCosmetics/Sounds/PickAxes/Hightower_Date/Pickaxe_Hightower_Date_Impact_Cue
root=Mixer_0
nodes [
 {id=Mixer_0, children [Modulator_0,Modulator_1], properties {inputVolume [1,0.5]}}
 {id=Modulator_0, children [Random_0]}
 {id=Modulator_1, children [Random_1], properties {pitchMax=1.1, volumeMax=0.95, volumeMin=0.9}}
 {id=Random_0, children [WavePlayer_1,WavePlayer_2,WavePlayer_3,WavePlayer_4], properties {weights [1,1,1,1]}}
 {id=Random_1, children [WavePlayer_5,WavePlayer_6,WavePlayer_7,WavePlayer_8,WavePlayer_0], properties {weights [1,1,1,1,1]}}
 {id=WavePlayer_0, properties {soundWaveAssetPtr=/BRCosmetics/Sounds/PickAxes/Galileo_Ferry/Pickaxe_GalileoFerry_Impact_05}}
 {id=WavePlayer_1, properties {soundWaveAssetPtr=&/PA_Hightower_Date_Impact_04}}
 {id=WavePlayer_2, properties {soundWaveAssetPtr=&/PA_Hightower_Date_Impact_01}}
 {id=WavePlayer_3, properties {soundWaveAssetPtr=&/PA_Hightower_Date_Impact_02}}
 {id=WavePlayer_4, properties {soundWaveAssetPtr=&/PA_Hightower_Date_Impact_03}}
 {id=WavePlayer_5, properties {soundWaveAssetPtr=/BRCosmetics/Sounds/PickAxes/Galileo_Ferry/Pickaxe_GalileoFerry_Impact_01}}
 {id=WavePlayer_6, properties {soundWaveAssetPtr=/BRCosmetics/Sounds/PickAxes/Galileo_Ferry/Pickaxe_GalileoFerry_Impact_02}}
 {id=WavePlayer_7, properties {soundWaveAssetPtr=/BRCosmetics/Sounds/PickAxes/Galileo_Ferry/Pickaxe_GalileoFerry_Impact_03}}
 {id=WavePlayer_8, properties {soundWaveAssetPtr=/BRCosmetics/Sounds/PickAxes/Galileo_Ferry/Pickaxe_GalileoFerry_Impact_04}}
]
properties {
 attenuationSettings=/Game/Sounds/Attenuation/Fort_Pickaxe_Impact_Att
 audioPropertiesSheet=/BRCosmetics/Sounds/Settings/PropertySheets/Pickaxe/APS_Cosmetic_Pickaxe_Impact
 bPrimeOnLoad=true
 maxDistance=4300
 soundClassObject=&Fort_Impacts
 attenuationOverrides {
  bAttenuateWithLPF=true
  distanceAlgorithm=Logarithmic
  falloffDistance=800
  lPFRadiusMax=800
  lPFRadiusMin=400
  attenuationShapeExtents {x=300}
 }
 concurrencyOverrides {bLimitToOwner=true, maxCount=2, resolutionRule=StopOldest}
 concurrencySet [&Player_Melee_Weapon_Conc]
}
